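# M_FakeTranslucency is a material gym test material that imitates translucency through the MF_OpaqueFakeTranslucency material function, which by its name targets opaque rendering.
# Scalar parameters for BaseColor, Specular, Roughness, Opacity and Refraction feed a MakeMaterialAttributes node.
# Those attributes pass into the function with a constant Alpha of one, using the Default Lit shading model.
# The faking technique itself lives inside the referenced function.
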Material /Game/Global/Materials/MaterialGym/Materials/FakeTranslucency/M_FakeTranslucency
graph {
 properties {
  bUseMaterialAttributes=true
  pixelDepthOffsetMode=PDOM_Legacy
  refractionMethod=RM_IndexOfRefraction
  translucencyLightingMode=TLM_SurfacePerPixelLighting
  usedShadingModels=Default Lit
 }
 nodes [MaterialFunctionCall!(
   Material=MakeMaterialAttributes!(
    BaseColor=ScalarParameter() {defaultValue=0.0716, parameterName=BaseColor}
    Metallic
    Specular=ScalarParameter() {defaultValue=1, parameterName=Specular}
    Roughness=ScalarParameter() {parameterName=Roughness}
    Anisotropy
    EmissiveColor
    Opacity=ScalarParameter() {defaultValue=0.3586, parameterName=Opacity}
    OpacityMask
    Normal
    Tangent
    WorldPositionOffset
    SubsurfaceColor
    ClearCoat
    ClearCoatRoughness
    AmbientOcclusion
    Refraction=ScalarParameter() {defaultValue=2.4, parameterName=Refraction}
    CustomizedUVs_0
    CustomizedUVs_1
    CustomizedUVs_2
    CustomizedUVs_3
    CustomizedUVs_4
    CustomizedUVs_5
    CustomizedUVs_6
    CustomizedUVs_7
    PixelDepthOffset
    ShadingModel
    Displacement
   )
   Alpha=Constant() {r=1}
  ) {materialFunction=/Game/Global/MaterialFunctions/MF_OpaqueFakeTranslucency}]
}
